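; Gameplay ability that gives secondary ToyFox pilots a third-person view of the finisher while it runs.
; On activation it caches the avatar's player controller, adds a camera modifier to its camera manager, and applies a gameplay effect that hides the HUD.
; On end it disables and removes the secondary-player finisher camera modifier and removes the HUD effect by its stored handle.
; CanActivateAbility always returns true.
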(event Ability|EventActivateAbility
  ;[0 0.15 0.007819 0.5] ActivateAbility: Enable the finisher camera modifier and hide the HUD.
  (|CallParentFunction)
  (CallFunction|CachePlayerController)
  (Default|SetEnableCameraModifier true)
  (bind _returnvalue (Ability|ApplyGameplayEffectToOwner (Variables|Settings|GetHideHUDGameplayEffectClass)))
  (Variables|Transient|SetHideHUDGameplayEffectHandle _returnvalue))

(event Ability|EventOnEndAbility
  (params (bWasCancelled :type Boolean))
  ;[0.1302 0.06461 0 0.5] EndAbility: Disable the finisher camera modifer and enable the HUD.
  (Utilities|FlowControl|Sequence
    (:then_0
      (Default|SetEnableCameraModifier false)
      (Ability|RemoveGameplayEffectFromOwnerWithHandle (Variables|Transient|GetHideHUDGameplayEffectHandle)))
    (:then_1
      (|CallParentFunction bWasCancelled))))

(fn CachePlayerController
  ;[0 0.1127 0.15 0.5] Cache the player controller.
  (bind _outfortplayerpawn (CallFunction|GetAvatarasValidFortPlayerPawn)
    (:ValidFortPlayerPawn
      (Variables|Transient|SetPlayerController (Fortnite|GetFortPlayerControllerfromActor _outfortplayerpawn)))))

(fn CanActivateAbility
  (params (ActorInfo :type Gameplay%20Ability%20Actor%20Info%20Structure) (Handle :type Gameplay%20Ability%20Spec%20Handle%20Structure))
  (return 0 true))

(fn SetEnableCameraModifier
  (params (bEnable :type Boolean))
  ;[0.00065 0.01698 0.15 0.5] SetEnableCameraModifier: Enable the camera modifier that allows the secondary player pilots to witness the finisher in third person view.
  (bind _playercontroller (Variables|Transient|GetPlayerController)
    (:then
      (bind _playercameramanager (Variables|/Script/Engine.PlayerController|GetPlayerCameraManager _playercontroller)
        (:then
          (if bEnable
            (bind _returnvalue (Game|Player|AddNewCameraModifier _playercameramanager (Variables|Settings|GetCameraModifierClass)))
            (Variables|Transient|SetCameraModifier _returnvalue)
            (else
              (bind _asb_toy_fox_giant_camera_modifier_secondary_player_finisher (Utilities|Casting|CastToB_ToyFoxGiantCameraModifier_SecondaryPlayerFinisher (Variables|Transient|GetCameraModifier)))
              (Class|BToyFoxGiantCameraModifierParent|DisableandRemoveModifier _asb_toy_fox_giant_camera_modifier_secondary_player_finisher))))
        (:else)))
    (:else)))
